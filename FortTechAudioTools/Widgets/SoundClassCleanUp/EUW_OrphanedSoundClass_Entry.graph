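# Editor utility row widget for one Sound Class, apparently an orphaned one in a Sound Class cleanup tool.
# ApplySoundClass stores the passed Sound Class and fills the text blocks with its object name and path.
# Clicking the name button opens that Sound Class in its asset editor for editing.
# The row also exposes an AllowMovement checkbox, checked by default, which this widget's own graph never reads.
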EditorUtilityWidgetBlueprint /FortTechAudioTools/Widgets/SoundClassCleanUp/EUW_OrphanedSoundClass_Entry
graphs [EventGraph {nodes [
   "On Clicked (SoundClassButton)"() {is=title K2Node_ComponentBoundEvent}
   > OpenEditorForAssets!(
    execute
    self:Asset Editor Subsystem Object Reference=Get AssetEditorSubsystem()
    Assets:Array of Object References=Make Array("[0]":Object Reference=Get SoundClassSelection() -> (SoundClassSelection:Sound Class Object Reference)) -> (Array:Array of Object References)
    OpenedMethod:EAssetTypeActivationOpenedMethod Enum=Edit
   ) -> (then, ReturnValue:bool)
   n1 = ApplySoundClass!() -> (OutputDelegate, then, SoundClassSelection:Sound Class Object Reference) {is=title K2Node_CustomEvent}
   > n2 = Set SoundClassSelection(SoundClassSelection:Sound Class Object Reference=@n1.SoundClassSelection) -> (Output_Get:Sound Class Object Reference)
   > Set Text(Text="To Text (String)"(InString=GetObjectName(Object=@n2)), self=Get SoundClassName() -> (SoundClassName:Text Object Reference))
   > Set Text(Text="To Text (String)"(InString=Get Object Path String(Object=@n2)), self=Get Directory() -> (Directory:Text Object Reference))
  ]}]
variables [SoundClassSelection {type=SoundClass, default=null, class=~, graph_type=~, children=~, properties=~}]
widgets {
 info {rootWidgetClass=&VerticalBox, widgetCount=9}
 widgets [
  VerticalBox_1 {widgetClassPath=&VerticalBox, widget=VerticalBox_1}
  Border_0 {
   widgetClassPath=&Border
   parent=VerticalBox_1
   widget=Border_0
   properties {brushColor=<rgba 0.0833 0.0692 0.0616 1>}
   slot {class=VerticalBoxSlot}
  }
  HorizontalBox_32 {widgetClassPath=&HorizontalBox, parent=Border_0, widget=HorizontalBox_32, slot {class=BorderSlot}}
  VerticalBox_0 {
   widgetClassPath=&VerticalBox
   parent=HorizontalBox_32
   widget=VerticalBox_0
   slot {class=HorizontalBoxSlot, properties {horizontalAlignment=HAlign_Left, verticalAlignment=VAlign_Center, size {sizeRule=Fill, value=1}}}
  }
  SoundClassButton {widgetClassPath=&EditorUtilityButton, parent=VerticalBox_0, widget=SoundClassButton, slot {class=VerticalBoxSlot}}
  SoundClassName {
   widgetClassPath=&TextBlock
   bIsVariable=true
   parent=SoundClassButton
   widget=SoundClassName
   properties {text=SoundClassNAme}
   slot {class=ButtonSlot, properties {horizontalAlignment=HAlign_Fill, verticalAlignment=VAlign_Fill, padding=<ltrb 0 0 0 0>}}
  }
  Directory {
   widgetClassPath=&TextBlock
   bIsVariable=true
   parent=VerticalBox_0
   widget=Directory
   properties {
    text=Text Block
    font {
     bForceMonospaced=false
     bMaterialIsStencil=false
     fontMaterial=null
     fontObject=&BurbankSmall
     letterSpacing=0
     monospacedWidth=1
     size=15
     skewAmount=0
     typefaceFontName=Bold
     outlineSettings {
      bApplyOutlineToDropShadows=false
      bMiteredCorners=false
      bSeparateFillAlpha=false
      outlineMaterial=null
      outlineSize=0
      outlineColor=<rgba 0 0 0 1>
     }
    }
   }
   slot {class=VerticalBoxSlot, properties {horizontalAlignment=HAlign_Left}}
  }
  AllowMovement {
   widgetClassPath=&EditorUtilityCheckBox
   bIsVariable=true
   parent=HorizontalBox_32
   widget=AllowMovement
   properties {checkedState=Checked, horizontalAlignment=HAlign_Center}
   slot {class=HorizontalBoxSlot, properties {horizontalAlignment=HAlign_Center, verticalAlignment=VAlign_Center, size {sizeRule=Fill, value=1}}}
  }
  Spacer_87 {widgetClassPath=&Spacer, parent=VerticalBox_1, widget=Spacer_87, properties {size=<1 10>}, slot {class=VerticalBoxSlot}}
 ]
}
