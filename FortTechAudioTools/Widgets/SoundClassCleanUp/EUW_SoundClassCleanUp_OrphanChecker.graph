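# Editor utility widget that finds selected SoundClasses with orphaned parent references and adds them to a named collection, defaulting to Audio_Soundclasses_ReadyForRemoval.
# Check confirms the collection exists through Python, then rebuilds a scroll list of EUW_OrphanedSoundClass_Entry widgets from HasOrphanedParentReference hits and updates the count.
# Move passes each entry with AllowMovement checked to the add_assets_to_collection Python module, which is reloaded on every call.
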EditorUtilityWidgetBlueprint /FortTechAudioTools/Widgets/SoundClassCleanUp/EUW_SoundClassCleanUp_OrphanChecker
graphs [EventGraph {nodes [
   "On Clicked (CheckClasses)"() {is=title K2Node_ComponentBoundEvent}
   n1 = Get SourceEntryBox() -> (SourceEntryBox:Editor Utility Editable Text Object Reference)
   > n2 = Execute Python Script!(
    execute
    PythonScript="import importlib; import collection_exists as m; importlib.reload(m); result = m.execute({\"collection\":\r\n  collection}); exists = result[\"exists\"]"
    collection:String="To String (Text)"(InText=Get Text(self:Editable Text Object Reference=@n1))
   ) -> (then, ReturnValue, exists:bool=~) {inputs [collection], outputs [exists]}
   > Branch(Condition=@n2.exists)
   > ClearScroll()
   > n3 = GetSelectedAssetsOfClass(AssetClass=/Script/Engine.SoundClass)
   > n4 = For Each Loop(Array:Array of Object References=@n3) -> (Array Element:Object Reference)
   n5 = GetAllChildren(self=Get SoundClassSelections() -> (SoundClassSelections:Editor Utility Scroll Box Object Reference))
   Set Text(
    execute=@n4.Completed
    Text="To Text (Integer)"(Value=Length(TargetArray:Array of Widget Object References=@n5))
    self=Get CountText() -> (CountText:Text Object Reference)
   )
   n6 = Cast To SoundClass!(execute=@n4.LoopBody, Object=@n4) -> (then, CastFailed, AsSound Class:Sound Class Object Reference) {is=title K2Node_DynamicCast}
   > Branch(Condition=HasOrphanedParentReference!(InSoundClass:Sound Class Object Reference=@n6) -> (ReturnValue:bool))
   > n7 = Create Widget!(
    execute:Exec
    WorldContextObject:Object Reference
    Class:User Widget Class Reference=&/EUW_OrphanedSoundClass_Entry.EUW_OrphanedSoundClass_Entry_C
    OwningPlayer:Player Controller Object Reference
   ) -> (then:Exec, ReturnValue:EUW Orphaned Sound Class Entry Object Reference)
   > AddChild(self=Get SoundClassSelections(), Content=@n7)
   > ApplySoundClass!(execute, self:EUW Orphaned Sound Class Entry Object Reference=@n7, SoundClassSelection:Sound Class Object Reference=@n6) -> (then)
   "On Clicked (MoveClasses)"() {is=title K2Node_ComponentBoundEvent}
   > n8 = For Each Loop(Array:Array of Widget Object References=GetAllChildren(self=Get SoundClassSelections())) -> (Array Element:Widget Object Reference)
   > n9 = Cast To EUW_OrphanedSoundClass_Entry!(execute, Object=@n8) -> (then, CastFailed, AsEUW Orphaned Sound Class Entry:EUW Orphaned Sound Class Entry Object Reference) {is=title K2Node_DynamicCast}
   > Branch(Condition=IsChecked(self=Get AllowMovement!(self:EUW Orphaned Sound Class Entry Object Reference=@n9) -> (AllowMovement:Editor Utility Check Box Object Reference)))
   n10 = Get SourceEntryBox()
   > Execute Python Script!(
    execute
    PythonScript="import importlib; import add_assets_to_collection as m; importlib.reload(m); m.execute({\"collection\": collection,\r\n  \"asset_path\": asset_path})"
    collection="To String (Text)"(InText=Get Text(self=@n10))
    asset_path:String=Get Object Path String(Object=Get SoundClassSelection!(self:EUW Orphaned Sound Class Entry Object Reference=@n9) -> (SoundClassSelection:Sound Class Object Reference))
   ) -> (then, ReturnValue) {inputs [collection,asset_path]}
   ClearScroll() {is=title K2Node_CustomEvent}
   n11 = Get SoundClassSelections()
   > n12 = For Each Loop(Array=GetAllChildren(self=@n11))
   > RemoveChild(self=@n11, Content=@n12)
  ]}]
variables [
 DestCollectionName {type=string, default=""}
 SourceCollectionName {type=string, default=""}
]
widgets {
 info {rootWidgetClass=&VerticalBox, widgetCount=14}
 widgets [
  VerticalBox_130 {widgetClassPath=&VerticalBox, widget=VerticalBox_130}
  HorizontalBox_152 {widgetClassPath=&HorizontalBox, parent=VerticalBox_130, widget=HorizontalBox_152, slot {class=VerticalBoxSlot}}
  TextBlock_125 {
   widgetClassPath=&TextBlock
   parent=HorizontalBox_152
   widget=TextBlock_125
   properties {text=CollectionName}
   slot {class=HorizontalBoxSlot}
  }
  Border_1 {
   widgetClassPath=&Border
   parent=HorizontalBox_152
   widget=Border_1
   properties {brushColor=<rgba 0 0 0 1>, contentColorAndOpacity=<rgba 0.8281 0.8281 0.8281 1>}
   slot {class=HorizontalBoxSlot, properties {size {sizeRule=Fill, value=1}}}
  }
  SourceEntryBox {
   widgetClassPath=/Script/Blutility.EditorUtilityEditableText
   bIsVariable=true
   parent=Border_1
   widget=SourceEntryBox
   properties {
    text=Audio_Soundclasses_ReadyForRemoval
    widgetStyle {
     backgroundImageComposing {
      bIsDynamicallyLoaded=false
      drawAs=Border
      imageType=FullColor
      mirroring=NoMirror
      resourceName=../../../Engine/Content/Slate/Old/HyperlinkDotted.png
      resourceObject=null
      tiling=Both
      imageSize=<0 0>
      margin=<ltrb 0 0 0 0.1875>
      outlineSettings {
       bUseBrushTransparency=false
       roundingType=HalfHeightRadius
       width=0
       color {colorUseRule=UseColor_Specified, specifiedColor=<rgba 0 0 0 0>}
       cornerRadii=<0 0 0 0>
      }
      tintColor {colorUseRule=UseColor_Specified, specifiedColor=<rgba 1 1 1 1>}
      uVRegion {bIsValid=false, max=<0 0>, min=<0 0>}
     }
     backgroundImageSelected {
      bIsDynamicallyLoaded=false
      drawAs=Image
      imageType=NoImage
      mirroring=NoMirror
      resourceName=None
      resourceObject=null
      tiling=NoTile
      imageSize=<0 0>
      margin=<ltrb 0 0 0 0>
      outlineSettings {
       bUseBrushTransparency=false
       roundingType=HalfHeightRadius
       width=0
       color {colorUseRule=UseColor_Specified, specifiedColor=<rgba 0 0 0 0>}
       cornerRadii=<0 0 0 0>
      }
      tintColor {colorUseRule=UseColor_Specified, specifiedColor=<rgba 0 0.162 0.7454 1>}
      uVRegion {bIsValid=false, max=<0 0>, min=<0 0>}
     }
     caretImage {
      bIsDynamicallyLoaded=false
      drawAs=Image
      imageType=NoImage
      mirroring=NoMirror
      resourceName=None
      resourceObject=null
      tiling=NoTile
      imageSize=<0 0>
      margin=<ltrb 0 0 0 0>
      outlineSettings {
       bUseBrushTransparency=false
       roundingType=HalfHeightRadius
       width=0
       color {colorUseRule=UseColor_Specified, specifiedColor=<rgba 0 0 0 0>}
       cornerRadii=<0 0 0 0>
      }
      tintColor {colorUseRule=UseColor_Specified, specifiedColor=<rgba 1 1 1 1>}
      uVRegion {bIsValid=false, max=<0 0>, min=<0 0>}
     }
     colorAndOpacity {colorUseRule=UseColor_Foreground, specifiedColor=<rgba 1 0 1 1>}
     font {
      bForceMonospaced=false
      bMaterialIsStencil=false
      fontMaterial=null
      fontObject=null
      letterSpacing=0
      monospacedWidth=1
      size=18.75
      skewAmount=0
      typefaceFontName=Regular
      outlineSettings {
       bApplyOutlineToDropShadows=false
       bMiteredCorners=false
       bSeparateFillAlpha=false
       outlineMaterial=null
       outlineSize=0
       outlineColor=<rgba 0 0 0 1>
      }
     }
     hintColorAndOpacity {colorUseRule=UseColor_Foreground, specifiedColor=<rgba 1 0 1 1>}
     selectedBackgroundColor {colorUseRule=UseColor_Foreground, specifiedColor=<rgba 1 0 1 1>}
    }
   }
   slot {class=BorderSlot}
  }
  TextBox1 {
   widgetClassPath=&TextBlock
   parent=HorizontalBox_152
   widget=TextBox1
   properties {text=Asset Count:}
   slot {class=HorizontalBoxSlot}
  }
  CountText {
   widgetClassPath=&TextBlock
   bIsVariable=true
   parent=HorizontalBox_152
   widget=CountText
   properties {text=0}
   slot {class=HorizontalBoxSlot}
  }
  HorizontalBox_0 {widgetClassPath=&HorizontalBox, parent=VerticalBox_130, widget=HorizontalBox_0, slot {class=VerticalBoxSlot}}
  CheckClasses {
   widgetClassPath=&EditorUtilityButton
   parent=HorizontalBox_0
   widget=CheckClasses
   slot {class=HorizontalBoxSlot, properties {horizontalAlignment=HAlign_Center, size {sizeRule=Fill, value=1}}}
  }
  CheckSoundClasses {
   widgetClassPath=&TextBlock
   parent=CheckClasses
   widget=CheckSoundClasses
   properties {text=Check Sound Classes}
   slot {class=ButtonSlot}
  }
  MoveClasses {
   widgetClassPath=&EditorUtilityButton
   parent=HorizontalBox_0
   widget=MoveClasses
   slot {class=HorizontalBoxSlot, properties {horizontalAlignment=HAlign_Center, size {sizeRule=Fill, value=1}}}
  }
  TextBlock_279 {
   widgetClassPath=&TextBlock
   parent=MoveClasses
   widget=TextBlock_279
   properties {text=Move to Collection}
   slot {class=ButtonSlot, properties {verticalAlignment=VAlign_Fill, padding=<ltrb 0 0 0 0>}}
  }
  CanvasPanel_0 {
   widgetClassPath=&CanvasPanel
   parent=VerticalBox_130
   widget=CanvasPanel_0
   properties {bOverride_Cursor=true}
   slot {class=VerticalBoxSlot}
  }
  SoundClassSelections {
   widgetClassPath=&EditorUtilityScrollBox
   bIsVariable=true
   parent=CanvasPanel_0
   widget=SoundClassSelections
   slot {
    class=CanvasPanelSlot
    properties {layoutData {alignment=<0 0>, anchors {maximum=<0 0>, minimum=<0 0>}, offsets=<ltrb 0 0 1919.8198 1000.9709>}}
   }
  }
 ]
}
